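# Post-process AnimBlueprint for the CoolSlice Golf male head, adjusting head and neck bones on the incoming pose.
# Chained constraints tie head to neck_02 and neck_02 to neck_01, followed by an additive neck_02 rotation; a neck_01-to-spine_05 constraint sits unconnected.
# Each update stores a scaled head-versus-neck_02 rotation delta from reference pose.
# Customization load flags the GolfVolley helmet by face-part name; neither stored value visibly feeds the AnimGraph.
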AnimBlueprint /BRCosmetics/Characters/Player/Male/Medium/Heads/M_MED_CoolSlice_Golf_Head/Meshes/M_MED_CoolSlice_Golf_Head_AnimBP_Post
target_skeleton=/Game/Characters/Player/Common/Fortnite_Base_Head/Fortnite_Base_Head_Export_Skeleton
graphs [
 AnimGraph {nodes [
   "Constraint - neck_01 (P)"!(ComponentPose, bAlphaBoolEnabled:bool=True, ConstraintWeights_0=1) -> (Pose) {alphaInputType=Bool, boneToModify {boneName=neck_01}, constraintSetup [{targetBone {boneName=spine_05}}], constraintWeights [1]}
   n1 = "Constraint - head (P)"!(ComponentPose=Local To Component(LocalPose=Input Pose()), bAlphaBoolEnabled:bool=True, ConstraintWeights_0=1) -> (Pose) {alphaInputType=Bool, boneToModify {boneName=head}, constraintSetup [{targetBone {boneName=neck_02}}], constraintWeights [1]}
   n2 = "Constraint - neck_02 (P)"!(ComponentPose=@n1, bAlphaBoolEnabled:bool=True, ConstraintWeights_0=1) -> (Pose) {alphaInputType=Bool, boneToModify {boneName=neck_02}, constraintSetup [{targetBone {boneName=neck_01}}], constraintWeights [1]}
   Output Pose(Result=Component To Local(ComponentPose="Transform (Modify) Bone - Bone: neck_02"!(ComponentPose=@n2, Alpha=1, Rotation="0,0,0") -> (Pose) {rotationMode=BMM_Additive, rotationSpace=BCS_BoneSpace, boneToModify {boneName=neck_02}}))
  ]}
 EventGraph {nodes [
   Event BlueprintUpdateAnimation()
   n1 = BreakRotator(InRot=GetDeltaTransformFromRefPose!(self=GetOwnerSkeletalMeshComponent(), BoneName=head, BaseName=neck_02) -> (ReturnValue_Location:Vector, ReturnValue_Rotation:Rotator, ReturnValue_Scale:Vector).ReturnValue_Rotation)
   > Set neck02Rot(neck02Rot:Rotator=MakeRotator(
     Roll=float * float(A=@n1, B=0.3) {is=title K2Node_PromotableOperator}
     Pitch=float * float(A=@n1.Pitch, B=0.3)
     Yaw=float * float(A=@n1.Yaw, B=0.4)
    )) -> (Output_Get:Rotator)
   n2 = Event OnCharacterCustomizationLoaded()
   > Set FaceAccMesh(FaceAccMesh:Skeletal Mesh Component Budgeted Object Reference=Get Array Item with EFort Custom Part Type(Array=@n2.CustomMeshParts, PartType=Face)) -> (Output_Get:Skeletal Mesh Component Budgeted Object Reference)
   n3 = Get FaceAccMesh() -> (FaceAccMesh:Skeletal Mesh Component Budgeted Object Reference)
   > Set isWearingHelmet(isWearingHelmet:bool=MatchesWildcard(SourceString=GetDisplayName(Object=@n3), Wildcard=*M_MED_CoolSlice_GolfVolley_FaceAcc*, SearchCase=CaseSensitive)) -> (Output_Get:bool)
  ]}
]
variables [
 FaceAccMesh {type=SkeletalMeshComponentBudgeted}
 isWearingHelmet {type=boolean}
 neck02Rot {type=Rotator}
]
